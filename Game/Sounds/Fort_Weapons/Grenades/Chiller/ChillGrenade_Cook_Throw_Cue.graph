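# Sound cue for the chill grenade's cook and throw action, played as a quiet gadget item sound.
# A single wave player reuses the chill grenade pickup sample through a modulator that randomly raises its pitch on each play.
# It uses weapon reload attenuation and the gadget items sound class, with a concurrency override that prevents new instances once the limit is reached.
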SoundCue /Game/Sounds/Fort_Weapons/Grenades/Chiller/ChillGrenade_Cook_Throw_Cue
root=Modulator_0
nodes [
 {id=Modulator_0, children [WavePlayer_1], properties {pitchMax=1.3, pitchMin=1.2, volumeMax=1, volumeMin=1}}
 {id=WavePlayer_1, properties {soundWaveAssetPtr=&/ChillGrenade_Pickup_01}}
]
properties {
 attenuationSettings=&Fort_Weapon_Reload_Att
 bOverrideConcurrency=true
 maxDistance=2250
 soundClassObject=&Fort_Gadgets_Items
 volumeMultiplier=0.05
 concurrencyOverrides {resolutionRule=PreventNew}
}
