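# SoundCue that plays the drop sound for the Chiller grenade weapon.
# A single wave player feeds the ChillGrenade_Drop_01 wave straight to the output at reduced volume, with no randomization.
# It uses the shared weapon drop attenuation and the loot collect sound class.
# A concurrency override caps simultaneous instances and prevents new ones from starting once the limit is reached.
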SoundCue /Game/Sounds/Fort_Weapons/Grenades/Chiller/ChillGrenade_Drop_Cue
root=WavePlayer_1
nodes [{id=WavePlayer_1, properties {soundWaveAssetPtr=&/ChillGrenade_Drop_01}}]
properties {
 attenuationSettings=&Fort_Weapon_Drop_Att
 bOverrideConcurrency=true
 maxDistance=3000
 soundClassObject=&Fort_Loot_Collect
 volumeMultiplier=0.3446
 concurrencyOverrides {maxCount=3, resolutionRule=PreventNew}
}
