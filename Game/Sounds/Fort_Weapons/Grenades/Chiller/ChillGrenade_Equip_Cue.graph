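# Sound cue for equipping the Chiller grenade weapon, playing a single equip wave.
# A Modulator root passes one Wave Player through with pitch and volume pinned at unity, so playback has no random variation.
# It plays in the player foley sound class at slightly reduced volume, using close-range prop attenuation.
# A concurrency override prevents new instances from starting once the limit is reached.
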SoundCue /Game/Sounds/Fort_Weapons/Grenades/Chiller/ChillGrenade_Equip_Cue
root=Modulator_1
nodes [
 {id=Modulator_1, children [WavePlayer_0], properties {pitchMax=1, pitchMin=1, volumeMax=1, volumeMin=1}}
 {id=WavePlayer_0, properties {soundWaveAssetPtr=&/ChillGrenade_Equip_01}}
]
properties {
 attenuationSettings=/Game/Sounds/Attenuation/Fort_Prop_Close_Att
 bOverrideConcurrency=true
 maxDistance=1800
 soundClassObject=&Fort_Player_Foley
 volumeMultiplier=0.8
 concurrencyOverrides {resolutionRule=PreventNew}
}
