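# Sound cue for equipping the Chiller grenade, playing a close-range player foley sound.
# A Random node picks with equal weight among four IceSkates equip wave variations.
# It plays through the player foley sound class with close prop attenuation and a slight volume boost.
# A concurrency override prevents new instances from starting once the limit is reached.
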SoundCue /Game/Sounds/Fort_Weapons/Grenades/Chiller/ChillGrenade_Equip_Cue_02
root=Random_1
nodes [
 {id=Random_1, children [WavePlayer_0, WavePlayer_8, WavePlayer_9, WavePlayer_10], properties {weights [1,1,1,1]}}
 {id=WavePlayer_0, properties {soundWaveAssetPtr=&/IceSkates_Equip_01}}
 {id=WavePlayer_10, properties {soundWaveAssetPtr=&/IceSkates_Equip_04}}
 {id=WavePlayer_8, properties {soundWaveAssetPtr=&/IceSkates_Equip_02}}
 {id=WavePlayer_9, properties {soundWaveAssetPtr=&/IceSkates_Equip_03}}
]
properties {
 attenuationSettings=/Game/Sounds/Attenuation/Fort_Prop_Close_Att
 bOverrideConcurrency=true
 maxDistance=1800
 soundClassObject=&Fort_Player_Foley
 volumeMultiplier=1.2
 concurrencyOverrides {resolutionRule=PreventNew}
}
