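# Sound cue for the Chiller grenade explosion, routed to the large explosives sound class with concurrency that prevents new instances at the limit.
# Each mix layers a close blast and a slightly quieter distant blast, each with its own attenuation and volume modulation, plus pitch variation on the distant layer.
# A quality node gives the first level three random waves per layer; the other levels share a leaner mix of two close and one distant.
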SoundCue /Game/Sounds/Fort_Weapons/Grenades/Chiller/ChillGrenade_Explode_Cue
root=QualityLevel_0
nodes [
 {
  id=Attenuation_0
  children [Modulator_13]
  properties {attenuationSettings=&Fort_Weapon_Projectile_Medium_Mono_Athena_Att}
 }
 {id=Attenuation_1, children [Modulator_12], properties {attenuationSettings=&Fort_Weapon_Projectile_Close_Athena_Att}}
 {
  id=Attenuation_8
  children [Modulator_1]
  properties {attenuationSettings=&Fort_Weapon_Projectile_Medium_Mono_Athena_Att}
 }
 {id=Attenuation_9, children [Modulator_0], properties {attenuationSettings=&Fort_Weapon_Projectile_Close_Athena_Att}}
 {id=Mixer_0, children [Attenuation_9, Attenuation_8], properties {inputVolume [1,0.8]}}
 {id=Mixer_2, children [Attenuation_1, Attenuation_0], properties {inputVolume [1,0.8]}}
 {id=Modulator_0, children [Random_2], properties {volumeMax=1.2, volumeMin=1.1}}
 {id=Modulator_1, children [WavePlayer_15], properties {pitchMin=0.8, volumeMax=1.2, volumeMin=1.1}}
 {id=Modulator_12, children [Random_0], properties {volumeMax=1.2, volumeMin=1.1}}
 {id=Modulator_13, children [Random_1], properties {pitchMin=0.8, volumeMax=1.2, volumeMin=1.1}}
 {id=QualityLevel_0, children [Mixer_2,Mixer_0,Mixer_0,Mixer_0]}
 {id=Random_0, children [WavePlayer_3,WavePlayer_4,WavePlayer_5], properties {weights [1,1,1]}}
 {id=Random_1, children [WavePlayer_6,WavePlayer_7,WavePlayer_8], properties {weights [1,1,1]}}
 {id=Random_2, children [WavePlayer_9, WavePlayer_13], properties {weights [1,1]}}
 {id=WavePlayer_13, properties {soundWaveAssetPtr=&/ChillGrenade_Explode_02}}
 {id=WavePlayer_15, properties {soundWaveAssetPtr=&/ChillGrenade_Explode_Distant_01}}
 {id=WavePlayer_3, properties {soundWaveAssetPtr=&/ChillGrenade_Explode_01}}
 {id=WavePlayer_4, properties {soundWaveAssetPtr=&ChillGrenade_Explode_02}}
 {id=WavePlayer_5, properties {soundWaveAssetPtr=&/ChillGrenade_Explode_03}}
 {id=WavePlayer_6, properties {soundWaveAssetPtr=&ChillGrenade_Explode_Distant_01}}
 {id=WavePlayer_7, properties {soundWaveAssetPtr=&/ChillGrenade_Explode_Distant_02}}
 {id=WavePlayer_8, properties {soundWaveAssetPtr=&/ChillGrenade_Explode_Distant_03}}
 {id=WavePlayer_9, properties {soundWaveAssetPtr=&ChillGrenade_Explode_01}}
]
properties {
 attenuationSettings=/Game/Sounds/Attenuation/Fort_Explosion_Far_StW_Att
 bOverrideConcurrency=true
 maxDistance=11000
 soundClassObject=/Game/Sounds/FortSoundClassesAndModes/Fort_Explosives_Large_PSM
 volumeMultiplier=0.8491
 concurrencyOverrides {resolutionRule=PreventNew}
}
