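# Sound cue for the Chiller grenade in flight, mixing an in-air loop with throw one-shots.
# The loop layer pairs a looping wave with a randomly chosen "Weee" voice clip borrowed from the snowball launcher; the throw layer randomly picks one of three throw sounds.
# Each layer has its own projectile attenuation, nested under a root RPG-explosion attenuation.
# A concurrency override caps the cue at three simultaneous instances.
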SoundCue /Game/Sounds/Fort_Weapons/Grenades/Chiller/ChillGrenade_InAirLoop_Cue
root=Attenuation_1
nodes [
 {
  id=Attenuation_0
  children [Modulator_0]
  properties {attenuationSettings=/Game/Sounds/Attenuation/Fort_Weapon_Projectile_InAir_Launch_Throw_Att}
 }
 {id=Attenuation_1, children [Mixer_1], properties {attenuationSettings=/Game/Sounds/Attenuation/Fort_Explosion_RPG_Att}}
 {id=Attenuation_2, children [Mixer_0], properties {attenuationSettings=/Game/Sounds/Attenuation/Fort_Weapon_Projectile_InAir_Loop_Att}}
 {id=Delay_0, children [Random_0]}
 {id=Delay_1, children [Random_1]}
 {id=Mixer_0, children [Modulator_2,Modulator_1], properties {inputVolume [1,1]}}
 {id=Mixer_1, children [Attenuation_2, Attenuation_0], properties {inputVolume [1,1]}}
 {id=Modulator_0, children [Delay_1]}
 {id=Modulator_1, children [Delay_0], properties {pitchMax=1, pitchMin=1}}
 {id=Modulator_2, children [WavePlayer_3], properties {volumeMax=1, volumeMin=0.9}}
 {id=Random_0, children [WavePlayer_5,WavePlayer_2], properties {bShouldExcludeFromBranchCulling=true, weights [1,1]}}
 {id=Random_1, children [WavePlayer_0,WavePlayer_4,WavePlayer_1], properties {bShouldExcludeFromBranchCulling=true, weights [1,1,1]}}
 {id=WavePlayer_0, properties {soundWaveAssetPtr=&/ChillGrenade_Throw_01}}
 {id=WavePlayer_1, properties {soundWaveAssetPtr=&/ChillGrenade_Throw_03}}
 {id=WavePlayer_2, properties {soundWaveAssetPtr=/Game/Sounds/Fort_Weapons/Launchers/WPN_SnowBall_Launcher_VO_Weee_01}}
 {id=WavePlayer_3, properties {bLooping=true, soundWaveAssetPtr=&/ChillGrenade_InAirLoop_01}}
 {id=WavePlayer_4, properties {soundWaveAssetPtr=&/ChillGrenade_Throw_02}}
 {id=WavePlayer_5, properties {soundWaveAssetPtr=/Game/Sounds/Fort_Weapons/Launchers/WPN_SnowBall_Launcher_VO_Weee_13}}
]
properties {
 bOverrideConcurrency=true
 maxDistance=13000
 soundClassObject=&Fort_WeaponRanged_Fire_3P_PSM
 volumeMultiplier=1.25
 concurrencyOverrides {maxCount=3}
}
