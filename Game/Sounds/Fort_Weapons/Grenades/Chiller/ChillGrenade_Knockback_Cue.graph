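# Sound cue for the Chiller grenade's knockback effect, played as a ranged weapon sound.
# It picks one of three knockback waves at random with equal weight, lowers pitch slightly at random through a modulator, and starts after a short fixed delay.
# Close-range projectile attenuation shapes falloff, and a concurrency override prevents new instances once the limit is reached.
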SoundCue /Game/Sounds/Fort_Weapons/Grenades/Chiller/ChillGrenade_Knockback_Cue
root=Delay_0
nodes [
 {id=Delay_0, children [Modulator_1], properties {delayMax=0.05, delayMin=0.05}}
 {id=Modulator_1, children [Random_0], properties {pitchMax=1, pitchMin=0.85}}
 {id=Random_0, children [WavePlayer_1,WavePlayer_2,WavePlayer_3], properties {weights [1,1,1]}}
 {id=WavePlayer_1, properties {soundWaveAssetPtr=&/ChillGrenade_Knockback_01}}
 {id=WavePlayer_2, properties {soundWaveAssetPtr=&/ChillGrenade_Knockback_02}}
 {id=WavePlayer_3, properties {soundWaveAssetPtr=&/ChillGrenade_Knockback_03}}
]
properties {
 attenuationSettings=&Fort_Weapon_Projectile_Close_Athena_Att
 bOverrideConcurrency=true
 maxDistance=4950
 soundClassObject=&Fort_WeaponRanged
 volumeMultiplier=1
 concurrencyOverrides {resolutionRule=PreventNew}
}
