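# Sound cue that plays the pickup sound for the Chiller grenade.
# A single wave player feeds the root with one pickup sample, with volume lowered and pitch raised slightly.
# It uses the weapon pickup attenuation and loot collect sound class.
# A concurrency override blocks new instances rather than stopping existing ones once the limit is reached.
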SoundCue /Game/Sounds/Fort_Weapons/Grenades/Chiller/ChillGrenade_Pickup_Cue
root=WavePlayer_1
nodes [{id=WavePlayer_1, properties {soundWaveAssetPtr=&/ChillGrenade_Pickup_01}}]
properties {
 attenuationSettings=&Fort_Weapon_Pickup_Att
 bOverrideConcurrency=true
 maxDistance=3500
 pitchMultiplier=1.1
 soundClassObject=&Fort_Loot_Collect
 volumeMultiplier=0.3446
 concurrencyOverrides {resolutionRule=PreventNew}
}
