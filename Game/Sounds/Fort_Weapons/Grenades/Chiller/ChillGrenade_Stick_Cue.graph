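# Sound cue for the Chiller grenade sticking, built from Chiller launcher snowball samples and mixed as two layers.
# A pitch-varied ground thud plays undelayed, while one of three random activation-charge sounds follows after a short delay with its own pitch variation.
# Each layer overrides medium-explosion attenuation with custom curves and low-pass filtering, and the thud carries farther.
# The cue's concurrency override stops the oldest instance.
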SoundCue /Game/Sounds/Fort_Weapons/Grenades/Chiller/ChillGrenade_Stick_Cue
root=Mixer_1
nodes [
 {
  id=Attenuation_0
  children [Modulator_0]
  properties {
   attenuationSettings=/Game/Sounds/Attenuation/Fort_Explosion_Medium_Athena_Att
   bOverrideAttenuation=true
   attenuationOverrides {
    bAttenuateWithLPF=true
    distanceAlgorithm=Custom
    falloffDistance=5500
    hPFFrequencyAtMax=50
    lPFFrequencyAtMin=500
    lPFRadiusMin=50
    reverbDistanceMin=200
    reverbWetLevelMax=1
    attenuationShapeExtents {x=200}
    customAttenuationCurve {editorCurveData {keys [
       {interpMode=RCIM_Cubic, tangentMode=RCTM_User, time=0.05, value=1}
       {arriveTangent=-0.5291, interpMode=RCIM_Cubic, leaveTangent=-0.5291, tangentMode=RCTM_User, time=0.2069, value=0.151}
       <at 1 0>
      ]}}
   }
  }
 }
 {
  id=Attenuation_2
  children [Modulator_1]
  properties {
   attenuationSettings=&Fort_Explosion_Medium_Athena_Att
   bOverrideAttenuation=true
   attenuationOverrides {
    bAttenuateWithLPF=true
    distanceAlgorithm=Custom
    falloffDistance=9500
    hPFFrequencyAtMax=50
    lPFFrequencyAtMin=500
    lPFRadiusMax=5000
    lPFRadiusMin=50
    reverbDistanceMax=15000
    reverbDistanceMin=200
    reverbWetLevelMax=1
    reverbWetLevelMin=0.35
    attenuationShapeExtents {x=200}
    customAttenuationCurve {editorCurveData {keys [
       {arriveTangent=-3, interpMode=RCIM_Cubic, leaveTangent=-3, tangentMode=RCTM_User, time=0.05, value=1}
       {arriveTangent=-0.2738, interpMode=RCIM_Cubic, leaveTangent=-0.2738, tangentMode=RCTM_User, time=0.25, value=0.0468}
       <at 1 0>
      ]}}
   }
  }
 }
 {id=Delay_0, children [Attenuation_0], properties {delayMin=0.1}}
 {id=Mixer_1, children [Delay_0, Attenuation_2], properties {inputVolume [1,1]}}
 {id=Modulator_0, children [Random_1], properties {pitchMax=1.2, pitchMin=0.9, volumeMax=1, volumeMin=1}}
 {id=Modulator_1, children [WavePlayer_0], properties {pitchMax=1.5, pitchMin=0.75, volumeMax=1, volumeMin=1}}
 {id=Random_1, children [WavePlayer_2,WavePlayer_4,WavePlayer_5], properties {weights [1,1,1]}}
 {id=WavePlayer_0, properties {soundWaveAssetPtr=/Game/Sounds/Fort_Weapons/Launchers/Chiller/WPN_Chiller_Snowball_Prj_Ground_Thud_01}}
 {id=WavePlayer_2, properties {soundWaveAssetPtr=/Game/Sounds/Fort_Weapons/Launchers/Chiller/WPN_Chiller_Snowball_Prj_Activate_Charge_01}}
 {id=WavePlayer_4, properties {soundWaveAssetPtr=/Game/Sounds/Fort_Weapons/Launchers/Chiller/WPN_Chiller_Snowball_Prj_Activate_Charge_02}}
 {id=WavePlayer_5, properties {soundWaveAssetPtr=/Game/Sounds/Fort_Weapons/Launchers/Chiller/WPN_Chiller_Snowball_Prj_Activate_Charge_03}}
]
properties {
 bOverrideConcurrency=true
 maxDistance=9700
 soundClassObject=&Fort_WeaponRanged_Fire_3P_PSM
 volumeMultiplier=1
 concurrencyOverrides {resolutionRule=StopOldest}
}
