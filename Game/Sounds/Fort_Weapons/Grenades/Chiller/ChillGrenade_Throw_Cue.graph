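# Sound cue for the throw foley of the Chiller grenade.
# A single wave player feeds a Modulator at default settings, adding slight random pitch and volume variation per play.
# It is routed to the gadget items sound class and uses the shared gadget-placement attenuation.
# Its own concurrency override prevents new instances from starting once the limit is reached, rather than stopping playing ones.
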SoundCue /Game/Sounds/Fort_Weapons/Grenades/Chiller/ChillGrenade_Throw_Cue
root=Modulator_1
nodes [
 {id=Modulator_1, children [WavePlayer_7]}
 {id=WavePlayer_7, properties {soundWaveAssetPtr=&/ChillGrenade_ThrowFoley_01}}
]
properties {
 attenuationSettings=/Game/Sounds/Attenuation/Fort_GadgetPlace_Att
 bOverrideConcurrency=true
 maxDistance=2000
 soundClassObject=&Fort_Gadgets_Items
 volumeMultiplier=1
 concurrencyOverrides {resolutionRule=PreventNew}
}
